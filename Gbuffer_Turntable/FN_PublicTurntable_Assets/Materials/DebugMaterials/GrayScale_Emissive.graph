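# GrayScale_Emissive is an unlit debug material in the Gbuffer turntable assets, named as a grayscale emissive.
# Its graph holds a RayTracingQualitySwitch that picks a constant mid-gray value for normal rendering and zero for ray-traced rendering.
# The file does not show the switch wired to a material output.
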Material /Gbuffer_Turntable/FN_PublicTurntable_Assets/Materials/DebugMaterials/GrayScale_Emissive
graph {
 properties {pixelDepthOffsetMode=PDOM_Legacy, shadingModel=MSM_Unlit, usedShadingModels=Unlit}
 nodes [RayTracingQualitySwitch(Normal=Constant() {r=0.5}, RayTraced=Constant())]
}
