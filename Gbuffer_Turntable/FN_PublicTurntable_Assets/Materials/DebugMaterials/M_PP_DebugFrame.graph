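# M_PP_DebugFrame is a post-process debug material in the Gbuffer Turntable assets that compares the rendered frame against a reference image.
# It samples the FrameReference texture parameter at viewport UVs and subtracts it from PostProcessInput0 RGB.
# An Abs node turns the result into a per-channel absolute difference, so matching pixels evaluate to black and mismatches stand out.
# Blendable alpha output is enabled.
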Material /Gbuffer_Turntable/FN_PublicTurntable_Assets/Materials/DebugMaterials/M_PP_DebugFrame
graph {
 properties {blendableOutputAlpha=true, materialDomain=MD_PostProcess, pixelDepthOffsetMode=PDOM_Legacy, usedShadingModels=Default Lit}
 nodes [
  n1 = SceneTexture() {sceneTextureId=PPI_PostProcessInput0}
  n2 = ScreenPosition()
  Abs(None=Subtract(A=ComponentMask(None=@n1.Color) {b=true, g=true, r=true}, B=TextureSampleParameter2D(UVs=@n2.ViewportUV) {parameterName=FrameReference}.RGB))
  Constant()
 ]
}
