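# Unlit debug material for checking scalability levels through flat colors, in the turntable's debug materials.
# A Quality Switch returns red for Medium, yellow for High and green for Epic and Default, with Low falling back to Default.
# A separate nested lerp blends blue, red and green using a Custom Primitive Data scalar.
# No link between the two branches or to a material output is shown.
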Material /Gbuffer_Turntable/FN_PublicTurntable_Assets/Materials/DebugMaterials/ScalabilityCheck
graph {
 properties {pixelDepthOffsetMode=PDOM_Legacy, shadingModel=MSM_Unlit, usedShadingModels=Unlit}
 nodes [
  n1 = Constant3Vector() {constant=<rgba 1 0 0 0>}
  n2 = Constant3Vector() {constant=<rgba 0 1 0 0>}
  n3 = ScalarParameter() {bUseCustomPrimitiveData=true}
  LinearInterpolate(A=Constant3Vector() {constant=<rgba 0 0 1 0>}, B=LinearInterpolate(A=@n1, B=@n2, Alpha=Subtract(A=@n3)) {desc=Low Or Epic}, Alpha=@n3) {desc=Default}
  QualitySwitch(Default=@n2, High=Constant3Vector() {constant=<rgba 1 1 0 0>}, Medium=@n1, Epic=@n2)
 ]
}
